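# Unlit material for a UI shell environment backplate, drawing a procedural colored background without texture samples.
# Exponential radial gradients blend two base colors, then layer bottom and main highlights with tunable colors and opacities.
# An inverted, smoothstepped radial gradient adds a movable, scalable vignette.
# A smoothstepped vertical UV gradient blends this toward a tinted multiply of the pre-vignette color, followed by adjustable desaturation.
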Material /Game/UI/Foundation/Environment/Shell/Materials/M_UI_Backplate
graph {
 properties {pixelDepthOffsetMode=PDOM_Legacy, shadingModel=MSM_Unlit, usedShadingModels=Unlit}
 nodes [
  ComponentMask(None=TextureCoordinate()) {g=true}
  Constant3Vector() {constant=<rgba 1 0 0.7877 0>}
  n1 = VectorParameter() {parameterName=RGBA1, defaultValue=<rgba 0.0176 0.0908 0.314 1>}
  n2 = VectorParameter() {parameterName=RGBA0, defaultValue=<rgba 0.0685 0.3005 0.7379 1>}
  n3 = MaterialFunctionCall(
   UVs=TextureCoordinate() {uTiling=2.1983, vTiling=1.4}
   CenterPosition=Constant2Vector() {r=0.9862}
   Radius=Constant() {r=1.5}
   Density=Constant() {r=2.325}
  ) {materialFunction=&RadialGradientExponential}
  n4 = ScalarParameter() {defaultValue=0.3743, parameterName=Bottom Highlight Alpha, sliderMax=1}
  n5 = MaterialFunctionCall(
   UVs=TextureCoordinate() {vTiling=2}
   CenterPosition=Constant2Vector() {g=1.6614, r=0.645}
   Radius=Constant() {r=0.4706}
   Density=Constant() {r=1.353}
  ) {materialFunction=&RadialGradientExponential}
  n6 = TextureCoordinate() {uTiling=2.132, vTiling=1.4}
  n7 = Constant2Vector() {g=0.4423, r=0.6651}
  n8 = Constant() {r=0.6918}
  n9 = Constant() {r=1}
  n10 = LinearInterpolate(
   A=LinearInterpolate(
    A=LinearInterpolate(A=@n1.RGB, B=@n2.RGB, Alpha=@n3.RadialGradientExponential)
    B=VectorParameter() {parameterName=Bottom Highlight, defaultValue=<rgba 0.515 0.7817 1 1>}.RGB
    Alpha=Multiply(A=@n4, B=@n5.RadialGradientExponential)
   )
   B=VectorParameter() {parameterName=Highlight, defaultValue=<rgba 0.71 0.6585 0.6248 1>}.RGB
   Alpha=Multiply(
    A=ScalarParameter() {defaultValue=0.8716, parameterName=Highlight Alpha, sliderMax=1}
    B=MaterialFunctionCall(UVs=@n6, CenterPosition=@n7, Radius=@n8, Density=@n9) {materialFunction=&RadialGradientExponential}.RadialGradientExponential
   )
  )
  n11 = VectorParameter() {parameterName=Vignette, defaultValue=<rgba 0.0865 0.175 0.1272 1>}
  n12 = VectorParameter() {parameterName=V scale, defaultValue=<rgba 1 1 0 0>}
  n13 = MaterialFunctionCall(
   UVs=Multiply(A=TextureCoordinate(), B=ComponentMask(None=@n12.RGB) {g=true, r=true})
   CenterPosition=ComponentMask(None=VectorParameter() {parameterName=v position, defaultValue=<rgba 0.9284 0.6922 0 1>}.RGB) {g=true, r=true}
   Radius=ScalarParameter() {defaultValue=0.86, parameterName=V Radius}
   Density=ScalarParameter() {defaultValue=1.5, parameterName=v density}
  ) {materialFunction=&RadialGradientExponential}
  n14 = Multiply(
   A=ScalarParameter() {defaultValue=1, parameterName=Vingette Alpha, sliderMax=1}
   B=MaterialFunctionCall(Alpha=OneMinus(None=@n13.RadialGradientExponential), Min=Constant(), Max=Constant() {r=1}) {materialFunction=&SmoothStep}.Result
  )
  n15 = MaterialFunctionCall(Alpha=ComponentMask(None=TextureCoordinate()) {g=true}, Min=Constant(), Max=Constant() {r=1}) {materialFunction=&SmoothStep}
  n16 = Multiply(
   A=VectorParameter() {parameterName=BottomMult, defaultValue=<rgba 0.5207 0.992 1 1>}.RGB
   B=Multiply(A=ScalarParameter() {defaultValue=0.89, parameterName=BottomMult Alpha}, B=@n15.Result)
  )
  Desaturation(
   None=LinearInterpolate(A=LinearInterpolate(A=@n10, B=@n11.RGB, Alpha=@n14), B=Multiply(A=@n10, B=@n16), Alpha=@n15.Result)
   Fraction=ScalarParameter() {defaultValue=0.5, parameterName=Desaturation}
  )
  Multiply(A=@n10, B=Multiply(A=@n11.RGB, B=@n14))
 ]
}
