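# Default Lit floor material for the UI shell environment, fading brightness and specular through a radial floor mask.
# A screen-aligned radial gradient blends two scene background collection colours, with optional desaturation.
# Screen-space top and middle gradients are declared for shaping the colour, and the floor mask interpolates high and low brightness.
# Gradient noise perturbs the mask edge against banding, and roughness has a separate mobile value.
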Material /Game/UI/Foundation/Environment/Shell/Materials/M_UI_FloorShadow
graph {
 properties {bScreenSpaceReflections=true, opacityMaskClipValue=0.333, pixelDepthOffsetMode=PDOM_Legacy, usedShadingModels=Default Lit}
 nodes [
  n1 = MaterialFunctionCall(
   Radius=ScalarParameter() {defaultValue=0.25, parameterName=Floor-Radius Mask}
   Density=ScalarParameter() {defaultValue=1, parameterName=Floor-Density Mask}
  ) {materialFunction=&RadialGradientExponential}
  n2 = Saturate(None=@n1.RadialGradientExponential)
  n3 = ScalarParameter() {defaultValue=0.005, parameterName=QuantizationNoiseStrength}
  n4 = Multiply(
   A=Multiply(A=OneMinus(None=@n2), B=Noise!(World Position, FilterWidth) {levels=4, noiseFunction=NOISEFUNCTION_GradientTex3D, scale=6})
   B=@n3
  ) {constB=0.003}
  n5 = Saturate(None=Subtract(A=Subtract(A=@n1.RadialGradientExponential, B=@n4), B=@n3) {constB=0.003})
  LinearInterpolate(B=ScalarParameter() {defaultValue=1, group=IsFloor, parameterName=Floor-Specular}, Alpha=@n5)
  n6 = Multiply(A=NamedRerouteUsage(), B=NamedRerouteUsage())
  n7 = CollectionParameter() {collection=&/M_UI_SceneBackgroundMCP, parameterName=RGBA0}
  n8 = CollectionParameter() {collection=&M_UI_SceneBackgroundMCP, parameterName=RGBA1}
  n9 = MaterialFunctionCall() {materialFunction=&ScreenAlignedUVs}
  n10 = MaterialFunctionCall(Float2=MaterialFunctionCall() {materialFunction=&ScreenResolution}.Visible Resolution) {materialFunction=&BreakOutFloat2Components}
  n11 = Divide(A=@n10.R, B=@n10.R)
  n12 = Constant() {r=1}
  n13 = Multiply(
   A=Multiply(A=@n9."X 100%, Y 100%", B=AppendVector(A=@n11, B=@n12))
   B=ScalarParameter() {defaultValue=1, group=Radial Gradient, parameterName=Tile}
  )
  n14 = ScalarParameter() {defaultValue=2, desc="Only works if CrewLegacyUnsubscribed = 0", group=Radial Gradient, parameterName=Radius}
  n15 = ScalarParameter() {defaultValue=1.5, desc="Only works if CrewLegacyUnsubscribed = 0", group=Radial Gradient, parameterName=Density}
  n16 = Add(
   A=AppendVector(A=Multiply(A=@n11) {constB=0.5}, B=Constant() {r=0.5})
   B=AppendVector(
    A=ScalarParameter() {defaultValue=0.65, group=Radial Gradient, parameterName=x-Offset}
    B=ScalarParameter() {defaultValue=0.35, group=Radial Gradient, parameterName=y-Offset}
   )
  )
  n17 = Desaturation(
   None=LinearInterpolate(
    A=ComponentMask(None=@n7) {b=true, g=true, r=true}
    B=ComponentMask(None=@n8) {b=true, g=true, r=true}
    Alpha=MaterialFunctionCall(UVs=@n13, CenterPosition=@n16, Radius=@n14, Density=@n15) {materialFunction=&RadialGradientExponential}.RadialGradientExponential
   )
   Fraction=ScalarParameter() {parameterName=Desat}
  )
  n18 = ScalarParameter() {defaultValue=1.2, parameterName=Floor-Brightness-High}
  n19 = ScalarParameter() {defaultValue=0.5, parameterName=Floor-Brightness-Low}
  LinearInterpolate(B=Multiply(A=Multiply(A=@n6, B=@n17), B=LinearInterpolate(A=@n18, B=@n19, Alpha=@n5)), Alpha=@n5)
  n20 = OneMinus(None=@n9."X 100%, Y 100%")
  n21 = Multiply(A=Add(A=ComponentMask(None=@n20) {g=true}) {constB=0.125}) {constB=1.5}
  n22 = Add(A=ComponentMask(None=@n9."X 100%, Y 100%") {g=true}) {constB=0.15}
  NamedRerouteDeclaration(None=Multiply(A=Saturate(None=@n21), B=Saturate(None=@n22))) {name=TopGradient, nodeColor=<rgba 0 0 0 1>}
  n23 = Add(A=@n9."X 100%, Y 100%") {constB=-0.5}
  n24 = OneMinus(None=ComponentMask(None=Abs(None=@n23)) {r=true})
  NamedRerouteDeclaration(None=Saturate(None=Add(A=@n24) {constB=0.025})) {name=MidGradient, nodeColor=<rgba 0 0 0 1>}
  ScalarParameter() {defaultValue=1, group=IsFloor, parameterName=Floor-Metallic}
  ShadingPathSwitch(
   Default=ScalarParameter() {defaultValue=0.25, group=IsFloor, parameterName=Floor-Roughness}
   Mobile=ScalarParameter() {defaultValue=0.75, parameterName=Floor-Roughness_mobile}
  )
  VectorParameter() {parameterName=RGBA0, defaultValue=<rgba 0.0754 0.0964 0.5 1>}
  VectorParameter() {parameterName=RGBA1, defaultValue=<rgba 0.2356 0.0604 0.6445 1>}
 ]
}
