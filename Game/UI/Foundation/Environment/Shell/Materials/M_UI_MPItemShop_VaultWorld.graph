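# Vault world background material for the MP item shop UI, masked and lit, feeding its color into the shared VaultWorld transition function.
# A scalar blends two color sources, apparently the Page A texture and a procedural fallback gradient.
# The fallback tints screen-space gradients with two desaturatable collection colors, darkened by a radial floor gradient with separate mobile density.
# Screen UVs are scaled to fit a 16:9 texture.
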Material /Game/UI/Foundation/Environment/Shell/Materials/M_UI_MPItemShop_VaultWorld
graph {
 properties {bCastRayTracedShadows=false, blendMode=BLEND_Masked, pixelDepthOffsetMode=PDOM_Legacy, usedShadingModels=Default Lit}
 nodes [
  n1 = NamedRerouteUsage()
  n2 = MaterialFunctionCall(Float2=@n1) {materialFunction=&BreakOutFloat2Components}
  LinearInterpolate(A=AppendVector(A=OneMinus(None=@n2.R), B=@n2.G), B=@n1, Alpha=NamedRerouteUsage())
  MaterialFunctionCall!(Vault World Color=LinearInterpolate(
    A=NamedRerouteUsage()
    B=NamedRerouteUsage()
    Alpha=ScalarParameter() {defaultValue=1, group=0 - Global, parameterName=PageA IsFallbackGradient}
   )) {materialFunction=/Game/Environments/FrontEnd/Blueprints/ItemPreview/VaultWorldTransitions/MaterialFunctions/MF_UI_VaultWorld_Transition}
  n3 = NamedRerouteUsage()
  n4 = OneMinus(None=@n3)
  n5 = Multiply(A=Add(A=ComponentMask(None=@n4) {g=true}) {constB=0.125}) {constB=1.5}
  n6 = Add(A=ComponentMask(None=@n3) {g=true}) {constB=0.15}
  NamedRerouteDeclaration(None=Multiply(A=Saturate(None=@n5), B=Saturate(None=@n6))) {name=TopGradient, nodeColor=<rgba 0 0 0 1>}
  n7 = TextureSampleParameter2D(UVs=NamedRerouteUsage()) {
   group=Page A
   parameterName=PageA-Texture
   samplerSource=SSM_Wrap_WorldGroupSettings
   texture=/Game/Athena/UI/Frontend/Spatial/Art/T_Ui_Empty
  }
  NamedRerouteDeclaration(None=@n7.RGB) {name=PageA - Color, nodeColor=<rgba 0 0 0 1>}
  NamedRerouteDeclaration(None=ScalarParameter() {parameterName=Transtion-Amount, sliderMax=1}) {name=Page - Transition, nodeColor=<rgba 0 0 0 1>}
  n8 = Add(A=@n3) {constB=-0.5}
  n9 = OneMinus(None=ComponentMask(None=Abs(None=@n8)) {r=true})
  NamedRerouteDeclaration(None=Saturate(None=Add(A=@n9) {constB=0.025})) {name=MidGradient, nodeColor=<rgba 0 0 0 1>}
  n10 = Multiply(A=NamedRerouteUsage(), B=NamedRerouteUsage())
  n11 = CollectionParameter() {collection=&/M_UI_SceneBackgroundMCP, parameterName=RGBA0}
  n12 = CollectionParameter() {collection=&M_UI_SceneBackgroundMCP, parameterName=RGBA1}
  n13 = MaterialFunctionCall(Float2=MaterialFunctionCall() {materialFunction=&ScreenResolution}.Visible Resolution) {materialFunction=&BreakOutFloat2Components}
  n14 = Divide(A=@n13.R, B=@n13.R)
  n15 = Constant() {r=1}
  n16 = MaterialFunctionCall() {materialFunction=&ScreenAlignedUVs}
  n17 = Multiply(A=@n14) {constB=0.5}
  n18 = Constant() {r=0.5}
  n19 = ScalarParameter() {defaultValue=0.65, group=Page A, parameterName=Gradient Offset X}
  n20 = ScalarParameter() {defaultValue=0.35, group=Page A, parameterName=Gradient Offset Y}
  n21 = MaterialFunctionCall(
   UVs=Multiply(A=AppendVector(A=@n14, B=@n15), B=NamedRerouteDeclaration(None=@n16."X 100%, Y 100%") {name=Screen UV, nodeColor=<rgba 0 0 0 1>})
   CenterPosition=Add(A=AppendVector(A=@n17, B=@n18), B=AppendVector(A=@n19, B=@n20)) {desc=disconnected for now}
   Radius=Constant() {r=2}
   Density=Constant() {r=1.5}
  ) {materialFunction=&RadialGradientExponential}
  n22 = Desaturation(
   None=ComponentMask(None=LinearInterpolate(A=@n11, B=@n12, Alpha=@n21.RadialGradientExponential)) {b=true, g=true, r=true}
   Fraction=ScalarParameter() {parameterName=Desat}
  )
  n23 = Constant() {r=1.2}
  n24 = Constant() {r=0.5}
  n25 = TextureCoordinate()
  n26 = ScalarParameter() {defaultValue=0.25, group=6 - Floor, parameterName=Floor Radius}
  n27 = ShadingPathSwitch(
   Default=LinearInterpolate(
    B=ScalarParameter() {defaultValue=1, group=6 - Floor, parameterName=Floor Density}
    Alpha=ScalarParameter() {defaultValue=1, group=6 - Floor, parameterName=Floor Reveal Transition, sliderMax=1}
   )
   Mobile=ScalarParameter() {defaultValue=0.25, parameterName=FloorDensityMobile}
  )
  n28 = NamedRerouteDeclaration(None=Saturate(None=MaterialFunctionCall(UVs=@n25, Radius=@n26, Density=@n27) {materialFunction=&RadialGradientExponential}.RadialGradientExponential)) {name=Floor Gradient, nodeColor=<rgba 0 0 0 1>}
  NamedRerouteDeclaration(None=Multiply(A=Multiply(A=@n10, B=@n22), B=LinearInterpolate(A=@n23, B=@n24, Alpha=@n28))) {name=FallbackGradient, nodeColor=<rgba 0 0 0 1>}
  n29 = MaterialFunctionCall() {materialFunction=&ScreenResolution}
  n30 = ComponentMask(None=@n29.Visible Resolution) {g=true}
  n31 = ComponentMask(None=@n29.Visible Resolution) {r=true}
  n32 = AppendVector(A=Divide(A=@n30, B=@n31), B=Divide(A=@n31, B=@n30))
  n33 = Constant2Vector() {desc=texture size, g=1080, r=1920}
  n34 = ComponentMask(None=@n33) {r=true}
  n35 = ComponentMask(None=@n33) {g=true}
  n36 = AppendVector(A=Divide(A=@n34, B=@n35), B=Divide(A=@n35, B=@n34))
  n37 = MaterialFunctionCall(UVs=ScreenPosition().ViewportUV, Texture Scale=Saturate(None=Multiply(A=@n32, B=@n36))) {desc=Scale To Fit, materialFunction=&ScaleUVsByCenter}
  NamedRerouteDeclaration(None=@n37.UVs) {name=TextureUVs, nodeColor=<rgba 0 0 0 1>}
  NamedRerouteDeclaration(None=@n37.0-1 mask) {name=UV mask, nodeColor=<rgba 0 0 0 1>}
  NamedRerouteDeclaration(None=ScreenPosition().ViewportUV) {name=screen UVs, nodeColor=<rgba 0 0 0 1>}
 ]
}
