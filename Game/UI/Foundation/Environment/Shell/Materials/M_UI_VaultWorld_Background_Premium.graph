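# Masked background material for the Premium VaultWorld UI shell, with vehicle and shop settings.
# Background visuals blend a screen-space texture with the default Fortnite gradient via a fallback parameter.
# A fake-fog mask blends in tint derived partly from a scene-background parameter collection, disabled under ray tracing.
# An optional shop base-color override follows, then post-process calibration, blended against uncalibrated color, feeds the VaultWorld transition function.
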Material /Game/UI/Foundation/Environment/Shell/Materials/M_UI_VaultWorld_Background_Premium
graph {
 properties {bAutomaticallySetUsageInEditor=false, blendMode=BLEND_Masked, usedShadingModels=Default Lit}
 nodes [
  n1 = Divide(A=NamedRerouteUsage()) {constB=0.3}
  n2 = Multiply(A=ComponentMask(None=CollectionParameter() {collection=&/M_UI_SceneBackgroundMCP, parameterName=RGBA1}) {b=true, g=true, r=true}) {constB=0.7}
  n3 = NamedRerouteUsage()
  n4 = TextureSample() {
   samplerSource=SSM_Clamp_WorldGroupSettings
   samplerType=SAMPLERTYPE_Alpha
   texture=/Game/UI/Foundation/Environment/Shell/Materials/Textures/T_UI_Vaultworld_Background_Premium_FakeFog
  }
  n5 = Multiply(
   A=NamedRerouteDeclaration(None=@n4.R) {name=FakeFogMask, nodeColor=<rgba 0 0 0 1>}
   B=ScalarParameter() {defaultValue=1, group=Vehicle Settings, parameterName=Display Fog, sliderMax=1}
  )
  n6 = Constant()
  n7 = VectorParameter() {
   desc=Alpha is used to Lerp OverrideValue
   group=Shop Settings
   parameterName=BaseColorOverride
   channelNames {a=UseOverride?, b="", g="", r=""}
  }
  n8 = LinearInterpolate(
   A=LinearInterpolate(A=NamedRerouteUsage(), B=LinearInterpolate(A=@n1, B=@n2, Alpha=@n3), Alpha=RayTracingQualitySwitch(Normal=@n5, RayTraced=@n6))
   B=ComponentMask(None=@n7) {b=true, g=true, r=true}
   Alpha=LinearInterpolate(A=@n7.UseOverride?, Alpha=ScalarParameter() {group=Shop Settings, parameterName=IsUsingShopTexture}) {constB=0}
  )
  n9 = MaterialFunctionCall!(Vault World Color=@n8) {materialFunction=/Game/Environments/FrontEnd/Blueprints/ItemPreview/VaultWorldTransitions/MaterialFunctions/MF_UI_VaultWorld_PostProcessCalibration}
  MaterialFunctionCall!(Vault World Color=LinearInterpolate(A=@n9.Calibrated Color, B=@n8, Alpha=NamedRerouteUsage())) {materialFunction=/Game/Environments/FrontEnd/Blueprints/ItemPreview/VaultWorldTransitions/MaterialFunctions/MF_UI_VaultWorld_Transition}
  n10 = TextureSampleParameter2D(UVs=ScreenPosition().ViewportUV) {
   group=Background Settings
   parameterName=BackgroundTexture
   samplerSource=SSM_Wrap_WorldGroupSettings
   texture=/Game/Athena/UI/Frontend/Spatial/Art/T_Ui_Empty
  }
  n11 = ScalarParameter() {defaultValue=1, group=Background Settings, parameterName=UseFallbackGradient}
  NamedRerouteDeclaration(None=LinearInterpolate(
    A=@n10.RGB
    B=MaterialFunctionCall() {materialFunction=/FortUILibrary/Materials/Background/DefaultFortnite/MF_UI_Background_DefaultFortnite}.Color
    Alpha=NamedRerouteDeclaration(None=@n11) {name=Use Fallback, nodeColor=<rgba 0 0 0 1>}
   )) {name=Background Visuals, nodeColor=<rgba 0 0 0 1>}
 ]
}
