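# Material function for UI shell environment materials that generates animated, distorted smoke from a tiling texture.
# Two red-channel samples, panned vertically and diagonally over scaled time, are summed and saturated into a noise mask, output as Smoke Distortion.
# That mask, scaled by Distortion Amount, offsets a separately timed diagonal pan for the Smoke Distortion Final sample.
# The diagonal noise panner has no coordinate connected, so it uses default UVs.
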MaterialFunction /Game/UI/Foundation/Environment/Shell/Materials/MaterialFunctions/MF_UI_BP_SmokeDistortion
graph {nodes [
  n1 = Multiply(
   A=NamedRerouteUsage()
   B=FunctionInput(Preview=Constant() {r=0.025}) {bUsePreviewValueAsDefault=true, inputName=Noise Mask Time Offset, inputType=FunctionInput_Scalar, sortPriority=5}
  )
  n2 = Panner(Coordinate=NamedRerouteUsage(), Time=@n1, Speed=Constant2Vector() {g=1})
  n3 = NamedRerouteUsage()
  n4 = Panner(Time=@n1, Speed=Constant2Vector() {g=1, r=1})
  n5 = NamedRerouteUsage()
  n6 = Saturate(None=Add(
    A=TextureSample(UVs=@n2, Tex=@n3) {texture=/Game/Effects/Fort_Effects/Textures/T_Tile_Pattern_01}.R
    B=TextureSample(UVs=@n4, Tex=@n5) {texture=&T_Tile_Pattern_01}.R
   ))
  n7 = Multiply(
   A=@n6
   B=FunctionInput(Preview=Constant() {r=0.0125}) {bUsePreviewValueAsDefault=true, inputName=Distortion Amount, inputType=FunctionInput_Scalar, sortPriority=3}
  )
  n8 = Constant() {r=0.0125}
  n9 = Panner(
   Coordinate=NamedRerouteUsage()
   Time=Multiply(
    A=NamedRerouteUsage()
    B=FunctionInput(Preview=@n8) {bUsePreviewValueAsDefault=true, inputName=Distortion Time Factor, inputType=FunctionInput_Scalar, sortPriority=4}
   )
   Speed=Constant2Vector() {g=1, r=1}
  )
  FunctionOutput(None=TextureSample(UVs=Add(A=@n7, B=@n9), Tex=NamedRerouteUsage()) {texture=&T_Tile_Pattern_01}.R) {outputName=Smoke Distortion Final, sortPriority=1}
  FunctionOutput(None=@n6) {outputName=Smoke Distortion}
  NamedRerouteDeclaration(None=Time()) {name=Time Value, nodeColor=<rgba 0 0 0 1>}
  NamedRerouteDeclaration(None=FunctionInput(Preview=TextureCoordinate()) {bUsePreviewValueAsDefault=true, inputName=UV, inputType=FunctionInput_Vector2}) {name=UV, nodeColor=<rgba 0 0 0 1>}
  NamedRerouteDeclaration(None=FunctionInput(Preview=TextureObject() {texture=&T_Tile_Pattern_01}) {bUsePreviewValueAsDefault=true, inputName=Smoke Texture, inputType=FunctionInput_Texture2D, sortPriority=1}) {name=Smoke Texture, nodeColor=<rgba 0 0 0 1>}
 ]}
