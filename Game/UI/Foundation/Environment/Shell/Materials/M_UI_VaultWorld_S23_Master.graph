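# Master material for the Season 23 UI vault world shell background, with a static switch for floor use.
# Page A and Page B each build a circular two-color gradient, brightened by bokeh and overlaid with a circle-masked, border-clamped texture.
# A blended color feeds the floor setup function, with a Transtion-Amount parameter for page transitions.
# Shared functions supply squared UVs, animated bokeh and smoke distortion, while some rotated wipe-mask math ends unconnected.
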Material /Game/UI/Foundation/Environment/Shell/Materials/M_UI_VaultWorld_S23_Master
graph {
 properties {pixelDepthOffsetMode=PDOM_Legacy, usedShadingModels=Default Lit}
 nodes [
  Distance(A=Add(B=NamedRerouteUsage()), B=ScalarParameter() {defaultValue=0.5})
  MaterialFunctionCall(Value=Multiply(A=Time(), B=ScalarParameter() {defaultValue=0.25, parameterName=Param_3})) {materialFunction=&LinearSine}
  MaterialFunctionCall!(
   Color=LinearInterpolate(A=NamedRerouteUsage(), B=NamedRerouteUsage(), Alpha=NamedRerouteUsage())
   Color Intensity=ScalarParameter() {defaultValue=1, group=0 - Global, parameterName=Color Intensity}
   Floor Radius=ScalarParameter() {defaultValue=0.175, group=6 - Floor, parameterName=Floor Radius}
   Floor Density=ScalarParameter() {defaultValue=4, group=6 - Floor, parameterName=Floor Density}
   Floor Metallic=ScalarParameter() {group=6 - Floor, parameterName=Floor Metalic}
   Floor Specular=ScalarParameter() {group=6 - Floor, parameterName=Floor Spec}
   Floor Roughness=ScalarParameter() {defaultValue=1, group=6 - Floor, parameterName=Floor Roughness}
   IsFloor=StaticBoolParameter() {group=6 - Floor, parameterName=IsFloor}
  ) {materialFunction=/Game/UI/Foundation/Environment/Shell/Materials/MaterialFunctions/MF_UI_BP_FloorSetup}
  n1 = Multiply(A=Constant2Vector() {g=1.7778, r=1}, B=ScalarParameter() {defaultValue=0.5625, group=Page A, parameterName=PageA Texture Scale})
  n2 = NamedRerouteUsage()
  n3 = Add(
   A=NamedRerouteUsage()
   B=AppendVector(
    A=ScalarParameter() {group=Page A, parameterName=PageA Texture Offset X}
    B=ScalarParameter() {group=Page A, parameterName=PageA Texture Offset Y}
   )
  )
  n4 = Add(A=Multiply(A=OneMinus(None=@n1)) {constB=0.5}, B=Multiply(A=@n1, B=Add(A=@n2, B=@n3)))
  n5 = TextureSampleParameter2D(UVs=@n4) {group=Page A, parameterName=PageA-Texture, samplerSource=SSM_Wrap_WorldGroupSettings}
  n6 = MaterialFunctionCall!(UVs=@n4, Alpha=@n5.A) {materialFunction=/Game/UI/Foundation/Items/ClampAlphaToTransparentBorder}
  n7 = Add(A=@n4) {constB=-0.5}
  n8 = Multiply(A=Subtract(A=Abs(None=@n7)) {constB=0.05}) {constB=2.25}
  n9 = Saturate(None=OneMinus(None=ComponentMask(None=@n8) {r=true}))
  n10 = ScalarParameter() {group=Page A, parameterName=PageA Texture Mask Offset X}
  n11 = ScalarParameter() {defaultValue=0.25, group=Page A, parameterName=PageA Texture Mask Offset Y}
  n12 = Add(A=NamedRerouteUsage(), B=Add(A=NamedRerouteUsage(), B=AppendVector(A=@n10, B=@n11)))
  n13 = ScalarParameter() {defaultValue=0.25, group=Page A, parameterName=PageA Texture Mask Size}
  n14 = ScalarParameter() {defaultValue=0.75, group=Page A, parameterName=PageA Texture Mask Falloff Min}
  n15 = ScalarParameter() {defaultValue=2, group=Page A, parameterName=PageA Texture Mask Falloff Max}
  n16 = Saturate(None=Multiply(
    A=Multiply(A=@n6.Result, B=@n9)
    B=MaterialFunctionCall(UVs=@n12, Size=@n13, GlowMin=@n14, GlowMax=@n15) {materialFunction=&MF_UI_Circle}.Fill
   ))
  n17 = ScalarParameter() {group=Page A, parameterName=PageA Gradient Offset X}
  n18 = ScalarParameter() {defaultValue=-0.25, group=Page A, parameterName=PageA Gradient Offset Y}
  n19 = Add(A=NamedRerouteUsage(), B=Add(A=NamedRerouteUsage(), B=AppendVector(A=@n17, B=@n18)))
  n20 = ScalarParameter() {group=Page A, parameterName=PageA-Gradient-Size}
  n21 = ScalarParameter() {defaultValue=-1, group=Page A, parameterName=PageA Gradient Falloff Min}
  n22 = ScalarParameter() {defaultValue=2, group=Page A, parameterName=PageA Gradient Falloff Max}
  n23 = LinearInterpolate(
   A=VectorParameter() {group=Page A, parameterName=PageA-RBG1, defaultValue=<rgba 0.0437 0.0027 0.0033 1>}.RGB
   B=VectorParameter() {group=Page A, parameterName=PageA-RBG2, defaultValue=<rgba 0.7758 0.0015 0.0152 1>}.RGB
   Alpha=Saturate(None=MaterialFunctionCall(UVs=@n19, Size=@n20, GlowMin=@n21, GlowMax=@n22) {materialFunction=&MF_UI_Circle}.Fill)
  )
  n24 = LinearInterpolate(
   A=@n23
   B=Multiply(A=@n23, B=ScalarParameter() {defaultValue=20, parameterName=Bokeh Color Multiply})
   Alpha=Multiply(A=NamedRerouteUsage(), B=ScalarParameter() {parameterName=Bokeh Intensity})
  )
  NamedRerouteDeclaration(None=LinearInterpolate(
    A=@n24
    B=LinearInterpolate(A=@n24, B=@n5.RGB, Alpha=@n16)
    Alpha=ScalarParameter() {defaultValue=0.75, group=Page A, parameterName=PageA Texture Intensity}
   )) {name=PageA - Color, nodeColor=<rgba 0 0 0 1>}
  n25 = MaterialFunctionCall() {materialFunction=/Game/UI/Foundation/Environment/Shell/Materials/MaterialFunctions/MF_UI_BP_UvSetup}
  NamedRerouteDeclaration(None=@n25.Squared UV) {name=Squared UV, nodeColor=<rgba 0 0 0 1>}
  n26 = Multiply(A=Constant2Vector() {g=1.7778, r=1}, B=ScalarParameter() {defaultValue=0.5625, group=Page B, parameterName=PageB Texture Scale})
  n27 = NamedRerouteUsage()
  n28 = Add(
   A=NamedRerouteUsage()
   B=AppendVector(
    A=ScalarParameter() {group=Page B, parameterName=PageB Texture Offset X}
    B=ScalarParameter() {group=Page B, parameterName=PageB Texture Offset Y}
   )
  )
  n29 = Add(A=Multiply(A=OneMinus(None=@n26)) {constB=0.5}, B=Multiply(A=@n26, B=Add(A=@n27, B=@n28)))
  n30 = TextureSampleParameter2D(UVs=@n29) {group=Page B, parameterName=PageB-Texture, samplerSource=SSM_Wrap_WorldGroupSettings}
  n31 = MaterialFunctionCall(UVs=@n29, Alpha=@n30.A) {materialFunction=&ClampAlphaToTransparentBorder}
  n32 = Add(A=@n29) {constB=-0.5}
  n33 = Multiply(A=Subtract(A=Abs(None=@n32)) {constB=0.05}) {constB=2.25}
  n34 = Saturate(None=OneMinus(None=ComponentMask(None=@n33) {r=true}))
  n35 = ScalarParameter() {group=Page B, parameterName=PageB Texture Mask Offset X}
  n36 = ScalarParameter() {defaultValue=0.25, group=Page B, parameterName=PageB Texture Mask Offset Y}
  n37 = Add(A=NamedRerouteUsage(), B=Add(A=NamedRerouteUsage(), B=AppendVector(A=@n35, B=@n36)))
  n38 = ScalarParameter() {defaultValue=0.25, group=Page B, parameterName=PageB Texture Mask Size}
  n39 = ScalarParameter() {group=Page B, parameterName=PageB Texture Mask Falloff Min}
  n40 = ScalarParameter() {defaultValue=2, group=Page B, parameterName=PageB Texture Mask Falloff Max}
  n41 = Saturate(None=Multiply(
    A=Multiply(A=@n31.Result, B=@n34)
    B=MaterialFunctionCall(UVs=@n37, Size=@n38, GlowMin=@n39, GlowMax=@n40) {materialFunction=&MF_UI_Circle}.Fill
   ))
  n42 = ScalarParameter() {group=Page B, parameterName=PageB Gradient Offset X}
  n43 = ScalarParameter() {defaultValue=-0.25, group=Page B, parameterName=PageB Gradient Offset Y}
  n44 = Add(A=NamedRerouteUsage(), B=Add(A=NamedRerouteUsage(), B=AppendVector(A=@n42, B=@n43)))
  n45 = ScalarParameter() {group=Page B, parameterName=PageB-Gradient-Size}
  n46 = ScalarParameter() {defaultValue=-1, group=Page B, parameterName=PageB Gradient Falloff Min}
  n47 = ScalarParameter() {defaultValue=2, group=Page B, parameterName=PageB Gradient Falloff Max}
  n48 = LinearInterpolate(
   A=VectorParameter() {group=Page B, parameterName=PageB-RBG1, defaultValue=<rgba 0.0454 0.0047 0.1302 1>}.RGB
   B=VectorParameter() {group=Page B, parameterName=PageB-RBG2, defaultValue=<rgba 0.3515 0.1301 0.807 1>}.RGB
   Alpha=Saturate(None=MaterialFunctionCall(UVs=@n44, Size=@n45, GlowMin=@n46, GlowMax=@n47) {materialFunction=&MF_UI_Circle}.Fill)
  )
  n49 = LinearInterpolate(
   A=@n48
   B=Multiply(A=@n48, B=ScalarParameter() {defaultValue=20, parameterName=Bokeh Color Multiply})
   Alpha=Multiply(A=NamedRerouteUsage(), B=ScalarParameter() {parameterName=Bokeh Intensity})
  )
  NamedRerouteDeclaration(None=LinearInterpolate(
    A=@n49
    B=LinearInterpolate(A=@n49, B=@n30.RGB, Alpha=@n41)
    Alpha=ScalarParameter() {defaultValue=0.75, group=Page B, parameterName=PageB Texture Intensity}
   )) {name=PageB - Color, nodeColor=<rgba 0 0 0 1>}
  NamedRerouteDeclaration(None=ScalarParameter() {parameterName=Transtion-Amount}) {name=Page - Transition, nodeColor=<rgba 0 0 0 1>}
  NamedRerouteDeclaration(None=Time()) {name=Time, nodeColor=<rgba 0 0 0 1>}
  NamedRerouteDeclaration(None=@n25.Squared UV Center Offset) {name=Squared UV Center Offset, nodeColor=<rgba 0 0 0 1>}
  NamedRerouteDeclaration(None=ScalarParameter() {defaultValue=0.5, group=Glow-Centre, parameterName=Cnt_Glow_x-Offset}) {name=Hotspot X-Offset, nodeColor=<rgba 0 0 0 1>}
  NamedRerouteDeclaration(None=ScalarParameter() {defaultValue=1, group=fx-streaks, parameterName=bIsUseStreak}) {name=FX Vis Control, nodeColor=<rgba 0 0 0 1>}
  n50 = NamedRerouteUsage()
  n51 = VectorParameter() {group=4 - Bokeh, parameterName=Bokeh Color, defaultValue=<rgba 1 1 1 1>}
  n52 = ScalarParameter() {defaultValue=0.0312, group=4 - Bokeh, parameterName=Bokeh Opacity}
  n53 = Add(A=NamedRerouteUsage(), B=NamedRerouteUsage())
  n54 = Constant() {r=0.5}
  n55 = Power(Base=ComponentMask(None=Distance(A=@n53, B=@n54)) {r=true}, Exp=ScalarParameter() {defaultValue=2, parameterName=Bokeh Mask Power})
  n56 = ScalarParameter() {defaultValue=2, parameterName=Boken Mask Intensity}
  NamedRerouteDeclaration(None=Multiply(
    A=MaterialFunctionCall!(
     UV=@n50
     Bokeh Texture
     Bokeh Color=@n51.RGB
     Bokeh Opacity=@n52
     Bokeh Power
     Bokeh Scale
     Foreground Scale
     Background Scale
     Bokeh Angle
     Bokeh Speed
     Falloff Intensity
     Falloff Edge
    ) {materialFunction=/Game/UI/Foundation/Environment/Shell/Materials/MaterialFunctions/MF_UI_BP_Bokeh}.Result
    B=Multiply(A=@n55, B=@n56)
   )) {name=Bokeh, nodeColor=<rgba 0 0 0 1>}
  n57 = MaterialFunctionCall!(
   UV=NamedRerouteUsage()
   Smoke Texture
   Distortion Amount=ScalarParameter() {defaultValue=0.0125, group=5 - Smoke Distortion, parameterName=Smoke Amount, sortPriority=2}
   Distortion Time Factor=ScalarParameter() {defaultValue=0.0125, group=5 - Smoke Distortion, parameterName=Smoke Speed, sortPriority=3}
   Noise Mask Time Offset=ScalarParameter() {defaultValue=0.025, group=5 - Smoke Distortion, parameterName=Smoke Mask Offset Speed, sortPriority=4}
  ) {materialFunction=/Game/UI/Foundation/Environment/Shell/Materials/MaterialFunctions/MF_UI_BP_SmokeDistortion}
  n58 = ScalarParameter() {defaultValue=0.04, group=5 - Smoke Distortion, parameterName=Smoke Opacity, sliderMax=1, sortPriority=1}
  n59 = VectorParameter() {group=5 - Smoke Distortion, parameterName=Smoke Color, sortPriority=0, defaultValue=<rgba 0.5164 0.0727 1 1>}
  NamedRerouteDeclaration(None=Multiply(A=Multiply(A=@n57.Smoke Distortion Final, B=@n58), B=@n59.RGB)) {name=Smoke Distortion, nodeColor=<rgba 0 0 0 1>}
  NamedRerouteDeclaration(None=@n25.Center Point) {name=Center Point, nodeColor=<rgba 0 0 0 1>}
  n60 = NamedRerouteUsage()
  n61 = ScalarParameter() {defaultValue=0.0099, parameterName=Param_4}
  n62 = ScalarParameter() {parameterName=Transtion-Amount}
  n63 = ComponentMask(None=Add(A=Rotator(Coordinate=@n60, Time=@n61) {speed=6.2832}, B=LinearInterpolate(Alpha=@n62) {constA=-1})) {r=true}
  n64 = ScalarParameter() {defaultValue=1, parameterName=s}
  n65 = ScalarParameter() {defaultValue=8, parameterName=Param_1}
  Saturate(None=Multiply(A=Subtract(A=@n63, B=@n64), B=@n65) {constB=64})
  n66 = OneMinus(None=@n60)
  n67 = ScalarParameter() {defaultValue=2, parameterName=Param_2}
  n68 = ScalarParameter() {parameterName=Param_5}
  n69 = ComponentMask(None=Add(A=Rotator(Coordinate=@n66, Time=@n61) {speed=6.2832}, B=LinearInterpolate(A=@n67, B=@n68, Alpha=@n62) {constA=-1})) {r=true}
  Saturate(None=Multiply(A=Subtract(A=@n69, B=@n64), B=@n65) {constB=64})
 ]
}
