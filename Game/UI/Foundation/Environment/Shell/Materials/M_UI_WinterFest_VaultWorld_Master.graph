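# Master material for the WinterFest vault world UI backdrop, with an IsFloor switch adding radially masked floor color, specular and roughness.
# A material parameter collection drives present-color tint and wrapping-paper pattern selection.
# Screen-space radial gradients layer the tint, pattern, distorted noise and paper creasing.
# A DigitalRain scalar crossfades to a darkened, desaturated look with animated glyph rain, built from stepped panners and random UV offsets.
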Material /Game/UI/Foundation/Environment/Shell/Materials/M_UI_WinterFest_VaultWorld_Master
graph {
 properties {pixelDepthOffsetMode=PDOM_Legacy, usedShadingModels=Default Lit}
 nodes [
  Constant() {r=0.45}
  n1 = VectorParameter() {group=DigitalRain, parameterName=GlyphColor, defaultValue=<rgba 1.2396 6 2.4297 0>}
  LinearInterpolate(
   A=VectorParameter() {group=DigitalRain, parameterName=GlyphColorAlt, defaultValue=<rgba 1.3715 2 0.1146 0>}.RGB
   B=@n1.RGB
   Alpha=ComponentMask(None=NamedRerouteUsage()) {g=true}
  )
  n2 = VectorParameter() {group=OverlayGradient, parameterName=OverlayGradientInner_Red, defaultValue=<rgba 0.8333 0.0732 0.0608 1>}
  n3 = CollectionParameter() {collection=&/MC_Winterfest_VaultWorldColors, parameterName=RGBA_01}
  n4 = VectorParameter() {group=OverlayGradient, parameterName=OverlayGradientInner_Green, defaultValue=<rgba 0.2142 0.4375 0.1185 1>}
  n5 = Add(
   A=LinearInterpolate(B=@n2.RGB, Alpha=ComponentMask(None=@n3) {r=true})
   B=LinearInterpolate(B=@n4.RGB, Alpha=ComponentMask(None=@n3) {g=true})
  )
  n6 = VectorParameter() {group=OverlayGradient, parameterName=OverlayGradientInner_Blue, defaultValue=<rgba 0.0984 0.3387 0.5104 1>}
  n7 = LinearInterpolate(B=@n6.RGB, Alpha=ComponentMask(None=@n3) {b=true})
  n8 = VectorParameter() {group=OverlayGradient, parameterName=OverlayGradientInner_Silver, defaultValue=<rgba 0.3662 0.392 0.5208 1>}
  n9 = Multiply(A=ComponentMask(None=@n3) {g=true}, B=ComponentMask(None=@n3) {b=true}) {constB=0}
  n10 = VectorParameter() {group=OverlayGradient, parameterName=OverlayGradientInner_Purple, defaultValue=<rgba 0.5246 0.1781 0.6979 1>}
  n11 = Multiply(A=ComponentMask(None=@n3) {r=true}, B=ComponentMask(None=@n3) {b=true}) {constB=0}
  n12 = Add(A=LinearInterpolate(B=@n8.RGB, Alpha=Saturate(None=@n9)), B=LinearInterpolate(B=@n10.RGB, Alpha=Saturate(None=@n11)))
  n13 = VectorParameter() {group=OverlayGradient, parameterName=OverlayGradientInner_Gold, defaultValue=<rgba 0.6875 0.4085 0.0895 1>}
  n14 = ComponentMask(None=@n3) {r=true}
  n15 = ComponentMask(None=@n3) {g=true}
  n16 = LinearInterpolate(B=@n13.RGB, Alpha=Saturate(None=Multiply(A=@n14, B=@n15) {constB=0}))
  n17 = ComponentMask(None=@n3) {r=true}
  n18 = ComponentMask(None=@n3) {g=true}
  n19 = Add(A=Add(A=Multiply(A=@n17) {constB=0.5}, B=Multiply(A=@n18) {constB=0.5}), B=Multiply(A=ComponentMask(None=@n3) {b=true}) {constB=0.5})
  NamedRerouteDeclaration(None=LinearInterpolate(A=Add(A=@n5, B=@n7), B=Add(A=@n12, B=@n16), Alpha=Floor(None=@n19))) {name=PresentColor, nodeColor=<rgba 0 0.5781 1 1>}
  n20 = MaterialFunctionCall(Float2=MaterialFunctionCall() {materialFunction=&ScreenResolution}.Visible Resolution) {materialFunction=&BreakOutFloat2Components}
  n21 = AppendVector(A=Divide(A=@n20.R, B=@n20.R), B=Constant() {r=1})
  n22 = MaterialFunctionCall() {materialFunction=&ScreenAlignedUVs}
  NamedRerouteDeclaration(None=Multiply(A=Multiply(A=@n21, B=@n22."X 100%, Y 100%"), B=ScalarParameter() {defaultValue=1, parameterName=DR-Tile})) {name=DR-ScreenSpaceUV, nodeColor=<rgba 0 1 0.7891 1>}
  n23 = NamedRerouteUsage()
  n24 = ScalarParameter() {defaultValue=6, parameterName=DR-MainUVScale}
  n25 = Time()
  n26 = Multiply(A=NamedRerouteUsage(), B=Multiply(A=ScalarParameter() {defaultValue=6, parameterName=DR-MainUVScale}, B=Constant() {r=8}))
  n27 = Multiply(A=Time(), B=ScalarParameter() {defaultValue=0.1647, group=DigitalRain, parameterName=RainCharacterGlyphRowSpeed})
  n28 = Constant() {r=8}
  n29 = Divide(A=Ceil(None=Multiply(A=@n27, B=@n28)), B=@n28)
  n30 = Multiply(
   A=ScalarParameter() {defaultValue=0.365, group=DigitalRain, parameterName=RainGlyphSwitchSpeed}
   B=MaterialFunctionCall!(
    PrimaryTextureCoordinates=Panner(Coordinate=@n26, Time=@n29)
    2ndOrderTextureCoordinates
    RandomSeed=ScalarParameter() {defaultValue=4, group=DigitalRain, parameterName=RainGlyphShiftRandomSeed}
    NoiseDensityBias
    Bypass Noise Influence Crosstalk
   ) {materialFunction=/Game/Effects/Fort_Effects/Materials/_Functions/MF_RandomUVSample}.NoiseWithRange
  )
  n31 = Constant() {r=8}
  n32 = Ceil(None=Multiply(A=Multiply(A=@n25, B=@n30), B=@n31))
  n33 = ScalarParameter() {defaultValue=6, parameterName=DR-MainUVScale}
  n34 = TextureSample!(
   UVs=Panner(Coordinate=Multiply(A=@n23, B=@n24), Time=Divide(A=@n32, B=@n31)) {bFractionalPart=true, speedX=1}
   Tex
   "DDX(UVs)"=DDX(Value=@n33)
   "DDY(UVs)"=DDY(Value=@n33)
   Apply View MipBias
  ) {mipValueMode=TMVM_Derivative, texture=/Game/Weapons/WeaponSkins/Wraps/Textures/DigitalRain/T_DigitalRain_Flipbook_01_SubPixelEdge}
  n35 = Constant() {r=8}
  n36 = Constant() {r=1}
  n37 = Multiply(
   A=AppendVector(A=ScalarParameter() {defaultValue=6, parameterName=DR-MainUVScale}, B=Constant() {r=1})
   B=Multiply(A=NamedRerouteUsage(), B=AppendVector(A=@n35, B=@n36))
  )
  n38 = ScalarParameter() {defaultValue=8, group=DigitalRain, parameterName=RainRandomSeed}
  n39 = Multiply(A=Constant() {r=8}, B=ScalarParameter() {defaultValue=6, parameterName=DR-MainUVScale})
  n40 = Floor(None=Multiply(
    A=MaterialFunctionCall(PrimaryTextureCoordinates=@n37, RandomSeed=@n38) {materialFunction=&MF_RandomUVSample}.NoiseWithRange
    B=@n39
   ))
  n41 = Multiply(A=Constant() {r=8}, B=ScalarParameter() {defaultValue=6, parameterName=DR-MainUVScale})
  n42 = Multiply(A=Multiply(A=Time(), B=ScalarParameter() {defaultValue=-0.4, group=DigitalRain, parameterName=RainPanSpeed}), B=@n41)
  n43 = Panner(Coordinate=Add(A=@n37, B=Divide(A=@n40, B=@n39) {constB=8}), Time=Divide(A=Ceil(None=@n42), B=@n41)) {speedY=1}
  n44 = SphereMask(
   A=Frac(None=MaterialFunctionCall(Float2=@n43) {materialFunction=&BreakOutFloat2Components}.G)
   B=ScalarParameter() {defaultValue=1, group=DigitalRain, parameterName=RainGradientCenterPoint}
   Radius=ScalarParameter() {defaultValue=0.45, group=DigitalRain, parameterName=RainGradientRadius}
   Hardness=Constant()
  ) {desc=Final rain gradient}
  n45 = ScalarParameter() {defaultValue=6, group=DigitalRain, parameterName=RainGradientExponent}
  NamedRerouteDeclaration(None=Multiply(A=Multiply(A=@n34.RGB, B=@n1.RGB), B=Power(Base=@n44, Exp=@n45))) {name=DigitalRain, nodeColor=<rgba 0.7813 0 1 1>}
  n46 = MaterialFunctionCall(Float2=MaterialFunctionCall() {materialFunction=&ScreenResolution}.Visible Resolution) {materialFunction=&BreakOutFloat2Components}
  n47 = Max(A=@n46.R)
  n48 = Multiply(
   A=AppendVector(A=Constant() {r=1}, B=Divide(A=@n46.G, B=@n47))
   B=MaterialFunctionCall() {materialFunction=&ScreenAlignedUVs}."X 100%, Y 100%"
  )
  n49 = Multiply(A=@n48)
  n50 = Multiply(A=Time(), B=ScalarParameter() {defaultValue=0.025, parameterName=Noise-Mask-OffSetTime})
  n51 = Multiply(A=@n48) {constB=2}
  n52 = Add(
   A=TextureSample!(UVs=Panner(Coordinate=@n49, Time=@n50) {speedY=1}, Tex, Apply View MipBias) {texture=/Game/Effects/Fort_Effects/Textures/T_Tile_Pattern_01}.R
   B=TextureSample(UVs=Panner(Coordinate=@n51, Time=@n50) {speedX=1, speedY=1}) {texture=&T_Tile_Pattern_01}.R
  )
  n53 = Add(A=@n48, B=Multiply(A=Saturate(None=@n52), B=ScalarParameter() {defaultValue=0.125, parameterName=NoiseMask-DistortionAmount}))
  NamedRerouteDeclaration(None=ComponentMask(None=TextureSample(UVs=@n53) {texture=&T_Tile_Pattern_01}.RGB) {g=true, r=true}) {name=Noise-Distortion, nodeColor=<rgba 0 0.5781 1 1>}
  Panner(Coordinate=@n48) {speedX=0.25, speedY=0.25}
  n54 = ScalarParameter() {defaultValue=0.5, group=Global, parameterName=Global-Floor-Roughness}
  n55 = MaterialFunctionCall(
   UVs=TextureCoordinate()
   Radius=ScalarParameter() {defaultValue=0.25, group=Global, parameterName=Global-FloorMask-Radius}
   Density=ScalarParameter() {defaultValue=2, group=Global, parameterName=Global-FloorMask-Density}
  ) {materialFunction=&RadialGradientExponential}
  Reroute(None=StaticSwitchParameter(True=LinearInterpolate(B=@n54, Alpha=@n55.RadialGradientExponential) {constA=1}, False=Constant() {r=1}) {group=Global, parameterName=IsFloor})
  n56 = ComponentMask(None=NamedRerouteUsage()) {b=true, g=true, r=true}
  n57 = MaterialFunctionCall(Float3=MaterialFunctionCall!(In=@n56) {materialFunction=/Engine/Functions/Engine_MaterialFunctions01/Chromakeying/RGBtoHSV}.Result) {materialFunction=&BreakOutFloat3Components}
  n58 = ScalarParameter() {defaultValue=0.2, parameterName=RGB_Saturation_Multiplier}
  n59 = MaterialFunctionCall!(HSV_In=MaterialFunctionCall(X=@n57.R, Y=@n57.R, Z=Multiply(A=@n57.B, B=@n58)) {materialFunction=&MakeFloat3}.Result) {materialFunction=/Game/Athena/UI/Materials/MF_HSVtoRGB}
  n60 = MaterialFunctionCall(Float2=MaterialFunctionCall() {materialFunction=&ScreenResolution}.Visible Resolution) {materialFunction=&BreakOutFloat2Components}
  n61 = Divide(A=@n60.R, B=@n60.R)
  n62 = Multiply(
   A=AppendVector(A=@n61, B=Constant() {r=1})
   B=MaterialFunctionCall() {materialFunction=&ScreenAlignedUVs}."X 100%, Y 100%"
  )
  n63 = ScalarParameter() {defaultValue=1, parameterName=Tile}
  n64 = AppendVector(A=ScalarParameter() {parameterName=x-Offset}, B=ScalarParameter() {parameterName=y-Offset})
  n65 = AppendVector(A=Multiply(A=@n61) {constB=0.5}, B=Constant() {r=0.5})
  n66 = MaterialFunctionCall(
   UVs=NamedRerouteDeclaration(None=Multiply(A=@n62, B=@n63)) {name=Radial-Gradient-UV, nodeColor=<rgba 0.5937 1 0 1>}
   CenterPosition=NamedRerouteDeclaration(None=Add(A=@n65, B=@n64)) {name=Radial-Gradient-Center, nodeColor=<rgba 0.8047 1 0 1>}
   Radius=ScalarParameter() {defaultValue=1.05, parameterName=Radius}
   Density=ScalarParameter() {defaultValue=2.35, parameterName=Density}
  ) {materialFunction=&RadialGradientExponential}
  n67 = NamedRerouteDeclaration(None=@n66.RadialGradientExponential) {name=Central-Gradient, nodeColor=<rgba 0 0.0127 1 1>}
  n68 = ScalarParameter() {defaultValue=2.5, parameterName=Radial-Gradient}
  n69 = ScalarParameter() {defaultValue=12, parameterName=CentralGradient-Hotspot-Power}
  n70 = Add(
   A=@n56
   B=Multiply(
    A=Power(Base=@n66.RadialGradientExponential, Exp=@n69)
    B=ScalarParameter() {defaultValue=0.1, parameterName=CentralGradient-Hotspot-Multiply}
   )
  )
  n71 = ScalarParameter() {defaultValue=0.125, parameterName=Pattern_Radial_Min_Clamp}
  n72 = Power(Base=@n67, Exp=ScalarParameter() {defaultValue=0.95, parameterName=Pattern-RadialMask-Power})
  n73 = MaterialFunctionCall(Float2=MaterialFunctionCall() {materialFunction=&ScreenResolution}.Visible Resolution) {materialFunction=&BreakOutFloat2Components}
  n74 = Multiply(
   A=AppendVector(A=Divide(A=@n73.R, B=@n73.R), B=Constant() {r=1})
   B=MaterialFunctionCall() {materialFunction=&ScreenAlignedUVs}."X 100%, Y 100%"
  )
  n75 = Multiply(
   A=Rotator(Coordinate=@n74, Time=ScalarParameter() {defaultValue=0.025, group=TilingTexture, parameterName=PatternRotation}) {speed=6.2832}
   B=ScalarParameter() {defaultValue=4, group=TilingTexture, parameterName=PatternBackground-Tile}
  )
  n76 = Multiply(A=NamedRerouteUsage(), B=ScalarParameter() {defaultValue=0.08, parameterName=Pattern-Distortion-Multiplier}) {constB=0}
  n77 = MaterialFunctionCall(Float4=CollectionParameter() {collection=&MC_Winterfest_VaultWorldColors, parameterName=Background_Pattern}) {materialFunction=&BreakOutFloat4Components}
  n78 = DotProduct(
   A=TextureSample(UVs=Add(A=@n75, B=@n76)) {samplerType=SAMPLERTYPE_LinearColor, texture=/Game/Athena/UI/Frontend/WinterQuest/WF_Art/T_UI_DF_WrappingPapers}.RGB
   B=MaterialFunctionCall(X=@n77.R, Y=@n77.R, Z=@n77.R) {materialFunction=&MakeFloat3}.Result
  )
  n79 = Multiply(
   A=Multiply(A=Round(None=@n78), B=ScalarParameter() {defaultValue=0.25, parameterName=Pattern-Multiplier})
   B=OneMinus(None=ScalarParameter() {parameterName=DigitalRain})
  )
  n80 = LinearInterpolate(
   A=LinearInterpolate(A=@n59.RGB_Out, B=@n70, Alpha=Power(Base=@n67, Exp=@n68))
   B=@n70
   Alpha=LinearInterpolate(A=ScalarParameter() {defaultValue=-0.05, parameterName=Pattern-Min-Color}, B=Clamp(None=@n72, Min=@n71), Alpha=Saturate(None=@n79))
  )
  n81 = Constant() {r=1}
  n82 = LinearInterpolate(
   A=Desaturation(None=@n80, Fraction=ScalarParameter() {defaultValue=0.125, parameterName=Desaturate})
   B=Multiply(A=Desaturation(None=@n80, Fraction=@n81)) {constB=0.125}
   Alpha=ScalarParameter() {parameterName=DigitalRain}
  )
  n83 = OneMinus(None=Max(A=Power(Base=@n72) {constExponent=0.5}) {constB=0.6})
  n84 = Multiply(
   A=Power(Base=NamedRerouteUsage(), Exp=ScalarParameter() {defaultValue=0.5, group=TilingTexture, parameterName=Noise-Power})
   B=ScalarParameter() {defaultValue=1, parameterName=Noise-Multiplier}
  )
  n85 = Multiply(A=@n83, B=Clamp(None=Saturate(None=@n84)))
  n86 = ScalarParameter() {defaultValue=0.125, parameterName=PaperNoise-Tile-Multiplier}
  n87 = Rotator(Coordinate=@n74, Time=ScalarParameter() {defaultValue=0.125, parameterName=PaperNoise-Rotator}) {centerX=1, centerY=1, speed=6.2832}
  n88 = TextureSample(UVs=Multiply(A=@n87, B=ScalarParameter() {defaultValue=1, group=TilingTexture, parameterName=PaperNoise-01-Tile})) {samplerType=SAMPLERTYPE_Normal, texture=/Game/Global/Textures/Default/Noise/BentNoise_N}
  n89 = Multiply(A=@n86, B=@n88.G)
  n90 = TextureSample(UVs=Multiply(A=@n87, B=ScalarParameter() {defaultValue=-0.4484, parameterName=PaperNoise-02-Tile})) {samplerType=SAMPLERTYPE_Normal, texture=&BentNoise_N}
  n91 = Subtract(A=Power(Base=Add(A=@n89, B=@n90.G)) {constExponent=0.25}) {constB=0.125}
  n92 = Saturate(None=Multiply(A=@n83, B=Multiply(A=@n91)))
  n93 = NamedRerouteDeclaration(None=Add(A=Multiply(A=@n88.R, B=@n86), B=@n90.R)) {name=Paper-Offset, nodeColor=<rgba 1 0 0.9609 1>}
  n94 = Multiply(A=@n79, B=@n93)
  n95 = Multiply(A=Subtract(A=@n93) {constB=0.0125}) {constB=0.2}
  n96 = Multiply(A=Multiply(A=@n83, B=Add(A=@n94, B=@n95))) {constB=0.65}
  n97 = Add(
   A=Multiply(
    A=MaterialFunctionCall!(Base=@n82, Blend=@n85) {materialFunction=/Engine/Functions/Engine_MaterialFunctions03/Blends/Blend_SoftLight}.Result
    B=LinearInterpolate(Alpha=@n92) {constA=1, constB=0}
   )
   B=Clamp(None=Saturate(None=@n96))
  )
  n98 = Multiply(A=NamedRerouteUsage(), B=ScalarParameter() {defaultValue=0.125, parameterName=DistortionAmount})
  n99 = Constant2Vector() {g=-0.375}
  n100 = MaterialFunctionCall(
   UVs=Add(A=NamedRerouteUsage(), B=Add(A=@n98, B=@n99))
   CenterPosition=NamedRerouteUsage()
   Radius=ScalarParameter() {defaultValue=1.125, parameterName=Hotspot-Overlay-Radius}
   Density=ScalarParameter() {defaultValue=4, parameterName=Hotspot-Overlay-Density}
  ) {materialFunction=&RadialGradientExponential}
  n101 = OneMinus(None=@n100.RadialGradientExponential)
  n102 = NamedRerouteUsage()
  n103 = NamedRerouteUsage()
  n104 = NamedRerouteUsage()
  n105 = ScalarParameter() {defaultValue=1, parameterName=RG-Radius}
  n106 = ScalarParameter() {defaultValue=2, parameterName=RG-Density}
  n107 = Multiply(
   A=OneMinus(None=MaterialFunctionCall(UVs=@n103, CenterPosition=@n104, Radius=@n105, Density=@n106) {materialFunction=&RadialGradientExponential}.RadialGradientExponential)
   B=ScalarParameter() {defaultValue=1, parameterName=DR-Radial-Mask-Intensity}
  )
  n108 = Add(A=Multiply(A=Multiply(A=@n101, B=@n102), B=Saturate(None=@n107)), B=@n97)
  Reroute(None=LinearInterpolate(A=@n97, B=LinearInterpolate(A=@n108, B=@n97, Alpha=@n100.RadialGradientExponential), Alpha=ScalarParameter() {parameterName=DigitalRain}))
  Reroute(None=ScalarParameter() {group=Global, parameterName=Global-Floor-Metalic})
  n109 = Constant() {r=1}
  n110 = LinearInterpolate(
   A=@n59.RGB_Out
   B=Multiply(A=Desaturation(None=@n59.RGB_Out, Fraction=@n109)) {constB=0.125}
   Alpha=ScalarParameter() {parameterName=DigitalRain}
  )
  n111 = Constant()
  Reroute(None=StaticSwitchParameter(True=LinearInterpolate(B=@n110, Alpha=@n55.RadialGradientExponential), False=@n111) {group=Global, parameterName=IsFloor})
  n112 = ScalarParameter() {defaultValue=0.5, group=Global, parameterName=Global-Floor-Spec}
  Reroute(None=StaticSwitchParameter(True=LinearInterpolate(B=@n112, Alpha=@n55.RadialGradientExponential), False=@n111) {group=Global, parameterName=IsFloor})
  VectorParameter() {parameterName=DigitalRain-CentreHotspot, defaultValue=<rgba 0.033 0.033 0.033 1>}
  VectorParameter() {parameterName=ColorPaper1, defaultValue=<rgba 1 0 0 1>}
 ]
}
